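# Material function for TurboFueled player character hair that overrides BaseColor and EmissiveColor on incoming material attributes, with a Mask input.
# Scrolling world-aligned noise, sampled with local-space positions, forms a distortion term.
# A camera-facing falloff and offset drive a gradient-texture lookup for emissive color, alongside a Fresnel-weighted blend.
# As saved, the final emissive is multiplied by zero and BaseColor is a default black constant, so hair renders black without glow.
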MaterialFunction /BRCosmetics/Characters/Player/Male/Medium/Bodies/M_MED_TurboFueled/FX/MF_TurboFueled_Hair
graph {nodes [
  n1 = NamedRerouteUsage()
  n2 = NamedRerouteUsage()
  n3 = LinearInterpolate(
   A=Multiply(A=NamedRerouteUsage(), B=ComponentMask(None=@n1) {r=true})
   B=NamedRerouteUsage()
   Alpha=Multiply(A=Fresnel() {exponent=1}, B=Multiply(A=@n2) {constB=0.67})
  ) {constA=0.1}
  FunctionOutput(None=SetMaterialAttributes!(MaterialAttributes=NamedRerouteUsage(), EmissiveColor=Multiply(A=@n3) {constB=0}, BaseColor=Constant()) {attributeSetTypes [#1, #2]})
  n4 = Power(Base=NamedRerouteUsage(), Exp=ScalarParameter() {defaultValue=0.5, group=BucketKick, parameterName=DistortionNoiseExponent, sortPriority=14})
  NamedRerouteDeclaration(None=Multiply(
    A=ConstantBiasScale(None=@n4) {bias=-0.5, scale=2}
    B=ScalarParameter() {defaultValue=0.2, group=BucketKick, parameterName=DistortionNoiseStrength, sortPriority=10}
   )) {name=Distortion, nodeColor=<rgba 1 0.7031 0 1>}
  n5 = NamedRerouteUsage()
  n6 = DotProduct(A=CameraVectorWS(), B=PixelNormalWS())
  n7 = NamedRerouteDeclaration(None=Power(Base=Saturate(None=@n6), Exp=ScalarParameter() {defaultValue=3, group=BucketKick, parameterName=FresnelExponent, sortPriority=43})) {name=FresenlValue, nodeColor=<rgba 0.7344 0 1 1>}
  NamedRerouteDeclaration(None=Saturate(None=Add(A=@n5, B=@n7))) {name=GradientUVX, nodeColor=<rgba 0 1 0.0625 1>}
  n8 = GetMaterialAttributes(None=NamedRerouteDeclaration(None=FunctionInput() {inputName=Material, inputType=FunctionInput_MaterialAttributes}) {name=IN_Material, nodeColor=<rgba 1 0.5391 0 1>}) {attributeGetTypes [#2, #1]}
  NamedRerouteDeclaration(None=@n8.BaseColor) {name=IN_BaseColor, nodeColor=<rgba 0.4297 0 1 1>}
  n9 = TextureObject() {samplerType=SAMPLERTYPE_LinearColor, texture=&/T_TurboFueled_Body_FX}
  n10 = VertexInterpolator(VS=LocalPosition() {includedOffsets=ExcludeOffsets}.XYZ)
  n11 = NamedRerouteUsage()
  n12 = Time() {period=2}
  n13 = VectorParameter() {group=BucketKick, parameterName=DistortionNoiseSpeed, sortPriority=12, defaultValue=<rgba 0 0 -11 0>}
  n14 = MaterialFunctionCall(
   TextureObject=@n9
   TextureSize=ScalarParameter() {defaultValue=15, group=BucketKick, parameterName=DistortionNoiseTextureSize, sortPriority=12}
   WorldPosition=Add(A=LinearInterpolate(A=@n10.PS, B=@n11) {constAlpha=0.1}, B=Multiply(A=@n12, B=@n13.RGB))
   ProjectionTransitionContrast=NamedRerouteDeclaration(None=ScalarParameter() {defaultValue=0.5, group=BucketKick, parameterName=DistortionNoiseTransitionContrast, sortPriority=13}) {name=TransitionContrast, nodeColor=<rgba 1 0.9844 0 1>}
  ) {materialFunction=&WorldAlignedTexture}
  NamedRerouteDeclaration(None=Saturate(None=ComponentMask(None=@n14.XYZ Texture) {b=true, g=true})) {name=DistortionNoiseSample, nodeColor=<rgba 1 0 0.5391 1>}
  NamedRerouteDeclaration(None=VertexInterpolator(VS=LocalPosition() {includedOffsets=ExcludeOffsets, localOrigin=InstancePreSkinning}.XYZ).PS) {name=PreskinnedPos, nodeColor=<rgba 0 0.9766 1 1>}
  NamedRerouteDeclaration(None=@n8.EmissiveColor) {name=IN_EmissiveColor, nodeColor=<rgba 0 1 0.1094 1>}
  NamedRerouteDeclaration(None=FunctionInput() {bUsePreviewValueAsDefault=true, inputName=Mask, sortPriority=2, previewValue=<1 0 0 1>}) {name=TechArtMask, nodeColor=<rgba 0.0001 1 0.9763 1>}
  n15 = TextureSample(UVs=MaterialFunctionCall(X=NamedRerouteUsage(), Y=Constant() {r=0.5}) {materialFunction=&MakeFloat2}.Result) {
   samplerSource=SSM_Clamp_WorldGroupSettings
   texture=&/T_TurboFueled_Gradient
  }
  n16 = ScalarParameter() {defaultValue=1, group=BucketKick, parameterName=GradientExponent, sortPriority=50}
  NamedRerouteDeclaration(None=Multiply(A=Power(Base=@n15.RGB, Exp=@n16), B=ScalarParameter() {defaultValue=1.2, group=BucketKick, parameterName=EmissiveStrength, sortPriority=51})) {name=EmissiveOUT_1, nodeColor=<rgba 1 0.0938 0 1>}
 ]}
