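# Material function for the TurboFueled character that builds a stylized head hair color and sets emissive and darkened base color on incoming material attributes.
# Hair center uses a view-angle gradient lookup with tint; edges mix Fresnel gold and side tint.
# A sphere mask in pre-skinned space, computed through vertex interpolators, blends center and sides.
# Time-of-day functions give the sides separate night saturation and emissive intensity.
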MaterialFunction /BRCosmetics/Characters/Player/Male/Medium/Bodies/M_MED_TurboFueled/FX/MF_TurboFueled_HeadHairBlend
graph {nodes [
  Constant()
  n1 = GetMaterialAttributes(None=NamedRerouteUsage()) {attributeGetTypes [#1]}
  FunctionOutput(None=SetMaterialAttributes!(MaterialAttributes=NamedRerouteUsage(), EmissiveColor=NamedRerouteUsage(), BaseColor=Multiply(A=@n1.BaseColor) {constB=0.8}) {attributeSetTypes [#2, #1]})
  LinearInterpolate(A=VertexInterpolator(VS=NamedRerouteUsage()).PS, B=VertexNormalWS(), Alpha=ScalarParameter() {defaultValue=0.3, parameterName=NormalLerp})
  n2 = Multiply(
   A=Subtract(
    A=LocalPosition() {includedOffsets=ExcludeOffsets, localOrigin=InstancePreSkinning}.XYZ
    B=VectorParameter() {parameterName=SphereMaskPositionOffsetForHairCenter, defaultValue=<rgba 0 0 0.344 1>}.RGB
   )
   B=VectorParameter() {parameterName=SphereMaskScale, defaultValue=<rgba 1 1 1 0>}.RGB
  )
  NamedRerouteDeclaration(None=SphereMask(
    A=VertexInterpolator(VS=@n2).PS
    B=Constant3Vector()
    Radius=ScalarParameter() {defaultValue=6, parameterName=SphereMaskRadius}
    Hardness=ScalarParameter() {defaultValue=5, parameterName=Hardness}
   )) {name=SphereMask, nodeColor=<rgba 0 1 0.8125 1>}
  n3 = LocalPosition()
  n4 = PreSkinnedLocalBounds()
  n5 = Subtract(A=ComponentMask(None=Subtract(A=@n3.XYZ, B=@n4.Min)) {b=true}, B=ScalarParameter() {defaultValue=5, parameterName=PivotVerticalOffset})
  NamedRerouteDeclaration(None=Normalize(VectorInput=AppendVector(A=@n3.XY, B=@n5))) {name=CoreNormals, nodeColor=<rgba 0 0.4609 1 1>}
  n6 = FunctionInput() {inputName=Material, inputType=FunctionInput_MaterialAttributes}
  NamedRerouteDeclaration(None=GetMaterialAttributes(None=NamedRerouteDeclaration(None=@n6) {name=IN_Material, nodeColor=<rgba 1 0.5391 0 1>}) {attributeGetTypes [#1, #2]}.BaseColor) {name=IN_BaseColor, nodeColor=<rgba 0.4297 0 1 1>}
  n7 = VertexNormalWS()
  n8 = Power(Base=DotProduct(A=Normalize(VectorInput=@n7), B=CameraVectorWS()), Exp=ScalarParameter() {defaultValue=2, parameterName=FresnelExponent})
  NamedRerouteDeclaration(None=Multiply(
    A=TextureSample(UVs=@n8) {texture=&/T_TurboFueled_Hair_CornerGradient}.RGB
    B=VectorParameter() {parameterName=Tint, defaultValue=<rgba 1 1 1 0>}.RGB
   )) {name=HairCenter, nodeColor=<rgba 0 1 0.9531 1>}
  n9 = Multiply(A=NamedRerouteUsage()) {constB=0.1}
  n10 = MaterialFunctionCall(
   A=Constant3Vector() {constant=<rgba 0 0 0 1>}
   B=Constant3Vector() {constant=<rgba 0.8148 0.5906 0.0742 1>}
   C=Constant3Vector() {constant=<rgba 0.831 0.5649 0.0802 1>}
   Alpha=Fresnel() {exponent=4}
  ) {materialFunction=&Lerp_3Color}
  n11 = VectorParameter() {parameterName=HairSidesTintColor, defaultValue=<rgba 0.2696 0.0736 0.0633 1>}
  n12 = ComponentMask(None=NamedRerouteUsage()) {g=true}
  n13 = ComponentMask(None=NamedRerouteUsage()) {g=true}
  NamedRerouteDeclaration(None=LinearInterpolate(A=LinearInterpolate(A=@n9, B=@n10.Result), B=Multiply(A=@n11.RGB, B=@n12), Alpha=Multiply(A=@n13) {constB=1.5})) {name=HairEdges, nodeColor=<rgba 1 0.3047 0 1>}
  n14 = ScalarParameter() {parameterName=HairSidesSaturationDay}
  n15 = ScalarParameter() {defaultValue=-0.2, parameterName=HairSidesSaturationNight}
  n16 = ScalarParameter() {defaultValue=1, parameterName=HairSidesEmissiveIntensity}
  n17 = Multiply(
   A=Desaturation(
    None=NamedRerouteUsage()
    Fraction=MaterialFunctionCall(Value_Morning=@n14, Value_Day=@n14, Value_Evening=@n14, Value_Night=@n15) {materialFunction=&Scalar_Time_of_Day}.Result
   )
   B=MaterialFunctionCall(
    Value_Morning=@n16
    Value_Day=@n16
    Value_Evening=@n16
    Value_Night=ScalarParameter() {defaultValue=1, parameterName=HairSidesEmissiveIntensityNight}
   ) {materialFunction=&Scalar_Time_of_Day}.Result
  )
  n18 = NamedRerouteUsage()
  n19 = NamedRerouteUsage()
  n20 = Multiply(A=@n19, B=ScalarParameter() {defaultValue=2, parameterName=SphereMaskMultiplier})
  n21 = AppendVector(A=@n19, B=@n19)
  n22 = Multiply(A=Power(Base=TextureSample(UVs=@n21) {texture=&T_TurboFueled_Hair_CornerGradient}.RGB))
  NamedRerouteDeclaration(None=LinearInterpolate(A=LinearInterpolate(A=@n17, B=@n18, Alpha=@n20), B=LinearInterpolate(A=@n17, B=@n22, Alpha=@n20)) {constAlpha=0.3}) {name=FinalHairColor, nodeColor=<rgba 0 0.0156 1 1>}
  NamedRerouteDeclaration(None=FunctionInput() {bUsePreviewValueAsDefault=true, inputName=Mask, sortPriority=2, previewValue=<1 0 0 1>}) {name=TechArtMask, nodeColor=<rgba 0.0001 1 0.9763 1>}
 ]}
