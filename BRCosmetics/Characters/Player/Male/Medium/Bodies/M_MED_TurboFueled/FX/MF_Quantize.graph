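# MF_Quantize is a material function in the TurboFueled character's FX folder that snaps a scalar to discrete stepped levels.
# It multiplies the float input by the steps input, applies Ceil, then divides by steps, so values round up to the next increment.
# Steps feeds the math through a named reroute, and a Frac node on the float input does not reach the output.
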MaterialFunction /BRCosmetics/Characters/Player/Male/Medium/Bodies/M_MED_TurboFueled/FX/MF_Quantize
graph {nodes [
  n1 = FunctionInput() {inputName=float, inputType=FunctionInput_Scalar, previewValue=<0.1 0 0 1>}
  Frac(None=@n1)
  n2 = Multiply(A=@n1, B=NamedRerouteUsage()) {constB=30}
  FunctionOutput(None=Divide(A=Ceil(None=@n2), B=NamedRerouteUsage()) {constB=30})
  NamedRerouteDeclaration(None=FunctionInput() {inputName=steps, inputType=FunctionInput_Scalar, sortPriority=1, previewValue=<30 0 0 1>}) {name=steps, nodeColor=<rgba 0.1484 0 1 1>}
 ]}
