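# Material function for the TurboFueled character body FX that outputs an animated wave value, named as a horizontal UV offset.
# A vertical coordinate, frequency and phase (time-driven by default) form the animation coordinate.
# Frac and square-root math shape semicircular arcs, and a hard-step sign alternates them into a circle wave.
# Named reroutes link the stages, including an amplitude derived from a green-channel mask that apparently scales the wave.
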MaterialFunction /BRCosmetics/Characters/Player/Male/Medium/Bodies/M_MED_TurboFueled/FX/MF_UVOffsetX
graph {nodes [
  FunctionOutput(None=Multiply(A=NamedRerouteUsage(), B=NamedRerouteUsage()))
  NamedRerouteDeclaration(None=FunctionInput(Preview=TextureCoordinate()) {inputName=UV, inputType=FunctionInput_Vector2}) {name=UV, nodeColor=<rgba 1 0 0.9141 1>}
  NamedRerouteDeclaration(None=FunctionInput(Preview=Constant() {r=5}) {bUsePreviewValueAsDefault=true, inputName=frequency, inputType=FunctionInput_Scalar}) {name=freq, nodeColor=<rgba 0 1 0.7891 1>}
  n1 = Time()
  NamedRerouteDeclaration(None=FunctionInput(Preview=Multiply(A=@n1)) {bUsePreviewValueAsDefault=true, inputName=phase, inputType=FunctionInput_Scalar}) {name=phase, nodeColor=<rgba 0.1122 1 0 1>}
  n2 = ComponentMask(None=NamedRerouteUsage()) {g=true}
  n3 = NamedRerouteUsage()
  NamedRerouteDeclaration(None=Add(A=Multiply(A=@n2, B=@n3), B=NamedRerouteUsage())) {name=xanim, nodeColor=<rgba 0.1527 0.3434 1 1>}
  n4 = Frac(None=NamedRerouteUsage())
  NamedRerouteDeclaration(None=Subtract(A=Multiply(A=@n4) {constB=2})) {name=a, nodeColor=<rgba 1 0.8242 0 1>}
  n5 = NamedRerouteUsage()
  n6 = Multiply(A=@n5, B=@n5)
  NamedRerouteDeclaration(None=SquareRoot(None=Subtract(B=@n6))) {name=circle, nodeColor=<rgba 0 1 0.3906 1>}
  n7 = Divide(A=NamedRerouteUsage())
  n8 = Saturate(None=SmoothStep(Value=Frac(None=@n7)) {constMax=0.5, constMin=0.5})
  NamedRerouteDeclaration(None=Subtract(A=Multiply(A=@n8) {constB=2})) {name=sign, nodeColor=<rgba 1 0 0.5191 1>}
  NamedRerouteDeclaration(None=Multiply(A=NamedRerouteUsage(), B=NamedRerouteUsage())) {name=circlewave, nodeColor=<rgba 0 1 0.9444 1>}
  n9 = NamedRerouteUsage()
  NamedRerouteDeclaration(None=Multiply(A=ComponentMask(None=@n9) {g=true}) {constB=0.1}) {name=amplitude, nodeColor=<rgba 1 0.6669 0 1>}
 ]}
