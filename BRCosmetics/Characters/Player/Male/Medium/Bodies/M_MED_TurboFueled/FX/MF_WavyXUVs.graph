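# Material function in the TurboFueled character's FX folder that waves texture UVs side to side along X.
# Time is scaled and quantized for stepped motion, then drives the phase of MF_UVOffsetX, with inputs for speed, step count, wave frequency and strength.
# The offset is weighted by one minus V, so it fades toward V = 1, then added to U while V passes through unchanged.
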MaterialFunction /BRCosmetics/Characters/Player/Male/Medium/Bodies/M_MED_TurboFueled/FX/MF_WavyXUVs
graph {
 properties {description=Waves the X UVs side to side}
 nodes [
  n1 = OneMinus(None=ComponentMask(None=TextureCoordinate()) {g=true})
  n2 = NamedRerouteUsage()
  n3 = MaterialFunctionCall!(float=Multiply(A=Time(), B=NamedRerouteUsage()), steps=NamedRerouteUsage()) {materialFunction=&/MF_Quantize}
  n4 = TextureCoordinate()
  n5 = Multiply(
   A=Multiply(
    A=Power(Base=@n1) {constExponent=1}
    B=MaterialFunctionCall!(frequency=@n2, phase=@n3.Result, UV=@n4) {materialFunction=&/MF_UVOffsetX}.Result
   )
   B=NamedRerouteUsage()
  )
  n6 = MaterialFunctionCall(Float2=TextureCoordinate()) {materialFunction=&BreakOutFloat2Components}
  FunctionOutput(None=MaterialFunctionCall(X=Add(A=@n5, B=@n6.R), Y=@n6.G) {materialFunction=&MakeFloat2}.Result)
  NamedRerouteDeclaration(None=FunctionInput(Preview=Constant() {r=12}) {inputName=TimeQuantizationSteps, inputType=FunctionInput_Scalar}) {name=TimeQuantizationSteps, nodeColor=<rgba 0 0.6484 1 1>}
  NamedRerouteDeclaration(None=FunctionInput(Preview=Constant() {r=5}) {inputName=UVWaveFrequency, inputType=FunctionInput_Scalar, sortPriority=1}) {name=UVWaveFrequency, nodeColor=<rgba 1 0 0.0809 1>}
  NamedRerouteDeclaration(None=FunctionInput(Preview=Constant() {r=1}) {inputName=TimeSpeed, inputType=FunctionInput_Scalar, sortPriority=2}) {name=TimeSpeed, nodeColor=<rgba 0 1 0.6484 1>}
  NamedRerouteDeclaration(None=FunctionInput(Preview=Constant() {r=4}) {inputName=WaveStrength, inputType=FunctionInput_Scalar, sortPriority=3}) {name=WaveStrength, nodeColor=<rgba 1 0.0206 0.4097 1>}
 ]
}
