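# Translucent, unlit, two-sided Niagara material for jet flames in the TurboFueled character cosmetic's FX.
# Panning noise, with per-particle random speed, distorts axis-swapped UVs for a scrolling wispy mask combined with a procedural UV gradient.
# The eroded result blends border to core color, tinted by particle color and desaturated.
# It also drives boosted alpha, faded by UV edge, view angle and, except on Low and Medium quality, scene depth.
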Material /BRCosmetics/Characters/Player/Male/Medium/Bodies/M_MED_TurboFueled/FX/M_Turbofueled_Jetflames
graph {
 properties {
  bAutomaticallySetUsageInEditor=false
  bUsedWithNiagaraMeshParticles=true
  bUsedWithNiagaraSprites=true
  blendMode=BLEND_Translucent
  shadingModel=MSM_Unlit
  twoSided=true
  usedShadingModels=Unlit
 }
 nodes [
  n1 = MaterialFunctionCall(XYZ=TextureCoordinate()) {materialFunction=&Swizzle}
  n2 = VectorParameter() {
   group=Distortion
   parameterName=DistortionUVControl
   sortPriority=2
   channelNames {a=SpeedY, b=SpeedX, g=ScaleY, r=ScaleX}
   defaultValue=<rgba 0.2 0.5 1 0.25>
  }
  n3 = ComponentMask(None=@n2) {g=true, r=true}
  n4 = AppendVector(A=@n2.SpeedX, B=@n2.SpeedX)
  n5 = ParticleRandom()
  n6 = TextureSampleParameter2D(UVs=Panner(Coordinate=Multiply(A=@n1.YX, B=@n3) {constB=0.6}, Speed=Add(A=@n4, B=@n5)) {speedX=1, speedY=0.25}) {
   group=Distortion
   parameterName=DistortionTexture
   samplerType=SAMPLERTYPE_Grayscale
   sortPriority=1
   texture=/Game/Effects/Fort_Effects/Textures/Fire/T_WispyFireNoise_Soft_M
  }
  n7 = AppendVector(
   A=Constant()
   B=Multiply(A=@n6.R, B=ScalarParameter() {defaultValue=0.2, group=Distortion, parameterName=DistortionAmount, sortPriority=3}) {constB=0.2}
  )
  n8 = Add(A=@n1.YX, B=@n7)
  n9 = VectorParameter() {
   group=MainTexture
   parameterName=FlameTextureUVControl
   sortPriority=2
   channelNames {a=SpeedY, b=SpeedX, g=ScaleY, r=ScaleX}
   defaultValue=<rgba 0.25 1 0.6 0>
  }
  n10 = ComponentMask(None=@n9) {g=true, r=true}
  n11 = TextureSampleParameter2D(UVs=Panner(Coordinate=Multiply(A=@n8, B=@n10), Speed=AppendVector(A=@n9.SpeedX, B=@n9.SpeedX)) {speedX=0.6}) {
   group=MainTexture
   parameterName=MainTexture
   sortPriority=1
   texture=/BRCosmetics/Athena/Effects/Particles/Skydiving/NS_Contrail_Chainmail/T_Chainmail_WispyMaskFat
  }
  n12 = TextureCoordinate()
  n13 = Power(Base=ComponentMask(None=@n12) {g=true}) {constExponent=1}
  n14 = Multiply(A=ComponentMask(None=@n12) {r=true}) {constB=0.5}
  n15 = Multiply(A=Saturate(None=Multiply(A=@n13) {constB=2}), B=Power(Base=Sine(None=@n14)))
  n16 = TextureCoordinate()
  n17 = Power(Base=ComponentMask(None=Add(A=@n16, B=@n7)) {g=true}) {constExponent=3}
  n18 = Add(
   A=ChannelMaskParameter(None=@n11.RGB) {group=MainTexture, parameterName=ChannelMask, sortPriority=5}.R
   B=Multiply(A=Multiply(A=@n15, B=@n17), B=ScalarParameter() {defaultValue=1, parameterName=Param_1})
  )
  n19 = ScalarParameter() {parameterName=TextureSubtract}
  n20 = ComponentMask(None=@n16) {g=true}
  n21 = Subtract(
   A=Saturate(None=Subtract(A=@n18, B=@n19) {constB=0.5})
   B=Power(Base=OneMinus(None=@n20), Exp=ScalarParameter() {defaultValue=3}) {constExponent=3}
  )
  n22 = LinearInterpolate(
   A=VectorParameter() {group=Color, parameterName=BorderColor, sortPriority=1, defaultValue=<rgba 0.4271 0.0515 0 1>}.RGB
   B=VectorParameter() {group=Color, parameterName=CoreColor, sortPriority=2, defaultValue=<rgba 2 0.3435 0 1>}.RGB
   Alpha=Power(Base=Saturate(None=@n21)) {constExponent=3}
  )
  n23 = ParticleColor()
  MaterialFunctionCall(Emissive=Desaturation(None=Multiply(A=@n22, B=@n23.RGB), Fraction=ScalarParameter() {defaultValue=0.15, group=Color, parameterName=Desaturation, sortPriority=3})) {materialFunction=&MF_FX_EmissiveAdjustments}
  n24 = Saturate(None=Multiply(A=@n21, B=ScalarParameter() {defaultValue=3, group=Color, parameterName=AlphaBoost, sortPriority=4}) {constB=3})
  n25 = OneMinus(None=ComponentMask(None=TextureCoordinate()) {g=true})
  n26 = Multiply(
   A=Multiply(A=Multiply(A=@n23.A, B=@n24), B=SmoothStep(Value=@n25) {constMax=0.5})
   B=MaterialFunctionCall("FadeSensitivity (optional)"=ScalarParameter() {defaultValue=1, parameterName=FadeBasedOnViewAngle}) {materialFunction=&FadeBasedOnViewAngle}.Result
  )
  n27 = DepthFade(Opacity=@n26, FadeDistance=ScalarParameter() {defaultValue=5, parameterName=DepthFade})
  QualitySwitch(Default=@n27, Low=@n26, High=@n27, Medium=@n26, Epic=@n27)
 ]
}
